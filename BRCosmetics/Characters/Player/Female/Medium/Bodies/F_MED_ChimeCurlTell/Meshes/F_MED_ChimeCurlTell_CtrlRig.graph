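# Post-process style Control Rig for the female ChimeCurlTell character body, layering deform, skirt, wristband and body fixes in sequence.
# Skirt bones average both thighs, then blend corrective offsets from thigh-up animation attributes, using their product when both legs rise.
# A Kalman-smoothed custom-emote flag blends extra back-skirt offsets.
# Bracelets follow forearm twist and hand; crouching and upright parachute flags trigger fixed bone offsets, including skirt and neckchain.
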ControlRigBlueprint /BRCosmetics/Characters/Player/Female/Medium/Bodies/F_MED_ChimeCurlTell/Meshes/F_MED_ChimeCurlTell_CtrlRig
graphs [
 New Function_ContainedGraph {nodes [
   RigVMFunctionEntryNode()
   > RigVMFunction_Sequence = RigVMAggregateNode()
   &RigUnit_ParentConstraint(
    AdvancedSettings=(InterpolationType=Average,RotationOrderForFilter=XZY)
    Child=(Type=Bone,Name="base_skirt_f_m")
    Parents=((Item=(Type=Bone,Name="thigh_l"),Weight=0.5),(Item=(Type=Bone,Name="thigh_r"),Weight=0.5))
    Weight=1
    ExecutePin=@RigVMFunction_Sequence.A
   )
   > &RigUnit_ParentConstraint(
    AdvancedSettings=(InterpolationType=Average,RotationOrderForFilter=XZY)
    Child=(Type=Bone,Name="base_skirt_b_m")
    Parents=((Item=(Type=Bone,Name="thigh_l"),Weight=0.5),(Item=(Type=Bone,Name="thigh_r"),Weight=0.5))
    Weight=1
   )
   n1 = RigVMDispatchNode(BoneName=root, Default=1, Name=deformRig_r_thigh_up_alpha) {template="DISPATCH_RigDispatch_GetAnimAttribute(in Name,in BoneName,in Default,out Value,out Found)"}
   &RigUnit_ModifyTransforms(
    ItemToModify=((Item=(Type=Bone,Name="base_skirt_f_m"),Transform=(Rotation=(X=0.2737,Y=-0.1283,Z=-0.0368,W=0.9525),Translation=(X=-0.7721,Y=-7.7462,Z=-0.2519),Scale3D=(X=1,Y=1,Z=1))))
    WeightMinimum=0
    Weight=@n1.Value
    ExecutePin=@RigVMFunction_Sequence.B
   )
   n2 = RigVMDispatchNode(BoneName=root, Default=1, Name=deformRig_l_thigh_up_alpha) {template="DISPATCH_RigDispatch_GetAnimAttribute(in Name,in BoneName,in Default,out Value,out Found)"}
   > &RigUnit_ModifyTransforms(
    ItemToModify=((Item=(Type=Bone,Name="base_skirt_f_m"),Transform=(Rotation=(X=-0.2737,Y=0.1283,Z=-0.0368,W=0.9525),Translation=(X=-0.7721,Y=-7.7462,Z=0.2519),Scale3D=(X=1,Y=1,Z=1))))
    WeightMinimum=0
    Weight=@n2.Value
   )
   > &RigUnit_ModifyTransforms(
    ItemToModify=((Item=(Type=Bone,Name="base_skirt_f_m"),Transform=(Rotation=(X=0,Y=0,Z=0,W=1),Translation=(X=0,Y=14.412,Z=3.9649),Scale3D=(X=1,Y=1,Z=1))))
    WeightMinimum=0
    Weight=&RigVMFunction_MathDoubleMul(A=@n1.Value, B=@n2.Value).Result
   )
   n3 = &RigVMFunction_MathDoubleRemap(SourceMaximum=1, SourceMinimum=0.4, TargetMaximum=1, TargetMinimum=0, Value=@n1.Value)
   &RigUnit_ModifyTransforms(
    ItemToModify=((Item=(Type=Bone,Name="base_skirt_b_m"),Transform=(Rotation=(X=0.4226,Y=0,Z=0,W=0.9063),Translation=(X=-5.0093,Y=-10.3759,Z=-0.0183),Scale3D=(X=1,Y=1,Z=1))))
    WeightMinimum=0
    Weight=@n3.Result
    ExecutePin=@RigVMFunction_Sequence.C
   )
   n4 = &RigVMFunction_MathDoubleRemap(SourceMaximum=1, SourceMinimum=0.4, TargetMaximum=1, TargetMinimum=0, Value=@n2.Value)
   > &RigUnit_ModifyTransforms(
    ItemToModify=((Item=(Type=Bone,Name="base_skirt_b_m"),Transform=(Rotation=(X=-0.4226,Y=0,Z=0,W=0.9063),Translation=(X=-5.0096,Y=-10.3759,Z=6.1302),Scale3D=(X=1,Y=1,Z=1))))
    WeightMinimum=0
    Weight=@n4.Result
   )
   > &RigUnit_ModifyTransforms(
    ItemToModify=((Item=(Type=Bone,Name="base_skirt_b_m"),Transform=(Rotation=(X=-0.0337,Y=0.005,Z=0.1476,W=0.9885),Translation=(X=0.0008,Y=21.2138,Z=4.3737),Scale3D=(X=1,Y=1,Z=1))))
    WeightMinimum=0
    Weight=&RigVMFunction_MathDoubleMul(B=@n4.Result, A=@n3.Result).Result
   )
   Cast = &RigVMFunction_MathBoolToFloat(Value=RigVMVariableNode() {variable=chimecurltell_custom_emote_playing}.Value)
   KalmanFilter = &RigVMFunction_KalmanFloat(BufferSize=16, Value=@Cast.Result) {template="KalmanFilter::Execute(in Value,in BufferSize,out Result)"}
   AlphaInterp = &RigVMFunction_AlphaInterp(InterpSpeedDecreasing=10, InterpSpeedIncreasing=10, Scale=1, bClampResult=False, bInterpResult=False, Value=@KalmanFilter.Result)
   > &RigUnit_ModifyTransforms(
    ItemToModify="((Item=(Type=Bone,Name=\"base_skirt_b_l\"),Transform=(Rotation=(X=0,Y=0,Z=0,W=1),Translation=(X=-10,Y=0,Z=0),Scale3D=(X=1,Y=1,Z=1))),(Item=(Type=Bone,Name=\"base_skirt_b_m\"),Transform=(Rotation=(X=0,Y=0,Z=0,W=1),Translation=(X=-30,Y=0,Z=0),Scale3D=(X=1,Y=1,Z= ...(truncated)"
    WeightMinimum=0
    Weight=@AlphaInterp.Result
   )
   > RigVMFunctionReturnNode()
  ]}
 New Function_ContainedGraph {nodes [
   RigVMFunctionEntryNode()
   > &RigUnit_ParentConstraint(
    AdvancedSettings=(InterpolationType=Average,RotationOrderForFilter=XZY)
    Child=(Type=Bone,Name="bracelet_01_l")
    Parents=((Item=(Type=Bone,Name="lowerarm_twist_01_l"),Weight=0.5),(Item=(Type=Bone,Name="hand_l"),Weight=0.5))
    Weight=1
    bMaintainOffset=true
   )
   > &RigUnit_ParentConstraint(
    AdvancedSettings=(InterpolationType=Average,RotationOrderForFilter=XZY)
    Child=(Type=Bone,Name="bracelet_02_l")
    Parents=((Item=(Type=Bone,Name="lowerarm_twist_01_l"),Weight=0.5),(Item=(Type=Bone,Name="hand_l"),Weight=0.5))
    Weight=1
   )
   > RigVMFunctionReturnNode()
  ]}
 New Function_ContainedGraph {nodes [
   n1 = RigVMFunctionEntryNode()
   n2 = &RigVMFunction_ControlFlowBranch(ExecuteContext=@n1.ExecuteContext, Condition=RigVMVariableNode() {variable=Is crouching}.Value)
   n3 = &RigVMFunction_ControlFlowBranch(ExecuteContext=@n2.Completed, Condition=RigVMVariableNode() {variable=Parachute Opened And Standing Upright}.Value)
   RigVMFunctionReturnNode(ExecuteContext=@n3.Completed)
   &RigUnit_ModifyTransforms(
    ItemToModify="((Item=(Type=Bone,Name=\"None\"),Transform=(Rotation=(X=0,Y=0,Z=0,W=1),Translation=(X=0,Y=0,Z=0),Scale3D=(X=1,Y=1,Z=1))),(Item=(Type=Bone,Name=\"base_neckchain\"),Transform=(Rotation=(X=0,Y=0.032,Z=0,W=-0.9995),Translation=(X=0,Y=0,Z=0),Scale3D=(X=1,Y=1,Z=1))) ...(truncated)"
    Weight=1
    WeightMinimum=0
    ExecutePin=@n3.True
   )
   &RigUnit_ModifyTransforms(
    ItemToModify="((Item=(Type=Bone,Name=\"base_skirt_f_r\"),Transform=(Rotation=(X=-0.0493,Y=-0.0094,Z=-0.1399,W=-0.9889),Translation=(X=0,Y=0,Z=0),Scale3D=(X=1,Y=1,Z=1))),(Item=(Type=Bone,Name=\"base_skirt_f_l\"),Transform=(Rotation=(X=0,Y=0,Z=-0.0823,W=0.9966),Translation=(X ...(truncated)"
    Weight=1
    WeightMinimum=0
    ExecutePin=@n2.True
   )
  ]}
 RigVMFunctionLibrary {nodes [
   "Body Fixes " = RigVMCollapseNode(ExecuteContext=())
   skirt = RigVMCollapseNode(ExecuteContext=())
   wristband = RigVMCollapseNode(ExecuteContext=())
  ]}
 RigVMFunction_Sequence_1_ContainedGraph {nodes [
   n1 = RigVMFunctionEntryNode()
   RigVMFunction_Sequence = &RigVMFunction_Sequence(ExecuteContext=@n1.ExecuteContext)
   n2 = &RigVMFunction_Sequence(ExecuteContext=@RigVMFunction_Sequence.B)
   RigVMFunctionReturnNode(A=@RigVMFunction_Sequence.A, B=@n2.A, C=@n2.B)
  ]}
 RigVMModel {nodes [
   RigUnit_BeginExecution = &RigUnit_BeginExecution()
   FN_DeformRig_FNC = RigVMFunctionReferenceNode(
    DeformRig_ Curve="(deformRig_l_upperarm_up_alpha_72_#1=False,deformRig_r_upperarm_up_alpha_73_#2=False,deformRig_l_upperarm_fr_alpha_74_#3=False,deformRig_r_upperarm_fr_alpha_75_CB51B8 ...(truncated)"
    DeformRig_Bool="(upperarm_1_#4=True,elbow_3_#5=True,wrist_7_#6=True,groin_8_#7=True,glute_17_#8=True,knee_9_40C408324CA261 ...(truncated)"
    DeformRig_Debug=(upperarm_debug_18_#4=False,wrist_debug_20_#6=False,lat_debug_26_#9=False,groinGlute_debug_27_#7=False)
    DeformRig_Value="(upperarm_up_18_#10=(Rotation=(X=0.0,Y=-0.0,Z=0.0,W=1.0),Translation=(X=0.0,Y=0.0,Z=8.0),Scale3D=(X=0.70,Y=1.0,Z=1.0)),upperarm_fwd_fr_19_#11=(Rotation=(X=0.0,Y=-0.0,Z=0.0,W=1.0),Translation=(X=0.0, ...(truncated)"
    ExecuteContext=@RigUnit_BeginExecution.ExecutePin
   )
   > skirt = RigVMFunctionReferenceNode()
   > wristband = RigVMFunctionReferenceNode()
   > "Body Fixes " = RigVMFunctionReferenceNode()
  ]}
 RigVMModel Construction Event Graph {nodes [
   n1 = &RigUnit_PrepareForExecution()
   &RigUnit_HierarchyImportFromSkeleton(NameSpace=None, bIncludeCurves=False, bIncludeMeshSockets=False, bIncludeVirtualBones=True, ExecutePin=@n1.ExecutePin)
  ]}
]
variables [
 Is Targeting {type=bool, default=False}
 Is crouching {type=bool, default=False}
 Parachute Opened And Standing Upright {type=bool, default=False}
 chimecurltell_custom_emote_playing {type=bool, default=False}
]
