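# Physics asset for the F_MED_ChimeCurlTell female player body cosmetic, covering a dangling neck chain and the upper torso.
# Capsule bodies sit on three spine bones, the chain base and two dynamic chain bones, with large spine_04 and spine_05 capsules approximating the torso.
# The two chain links have tight swing and near-locked twist limits.
# Constraints joining the spine bones and chain base leave all angular motion free.
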PhysicsAsset /BRCosmetics/Characters/Player/Female/Medium/Bodies/F_MED_ChimeCurlTell/Meshes/F_MED_ChimeCurlTell_Physics
bodies {
 base_neckchain [{shapeType=Capsule, length=0.505, radius=0.505, shapeName=base_neckchain_capsule, rotation=<rot 0 -90 0>}]
 dyn_neckchain_01 [{shapeType=Capsule, length=2, radius=2, shapeName=dyn_neckchain_01_capsule, center=<2 0 1.3634>, rotation=<rot 90 0 0>}]
 dyn_neckchain_02 [
  {shapeType=Capsule, length=5, radius=2, shapeName=dyn_neckchain_02_capsule, center=<3.5755 -0.0057 1.79>, rotation=<rot 90 0 0>}
  {shapeType=Capsule, length=5, radius=2, shapeName=dyn_neckchain_02_capsule, center=<3.1789 0 1.7549>, rotation=<rot 90 90 0>}
 ]
 spine_03 [{shapeType=Capsule, length=0.505, radius=0.505, shapeName=spine_03_capsule, rotation=<rot 0 -90 0>}]
 spine_04 [{shapeType=Capsule, length=5, radius=10.5, shapeName=spine_04_capsule, center=<1.2658 1.9349 0>, rotation=<rot 0 -90 0>}]
 spine_05 [{shapeType=Capsule, length=13, radius=9, shapeName=spine_05_capsule, center=<0 -0.5355 0>}]
}
constraints [
 {bone1Name=dyn_neckchain_01, bone2Name=base_neckchain, swing1LimitDegrees=20, swing2LimitDegrees=10, twistLimitDegrees=1}
 {bone1Name=dyn_neckchain_02, bone2Name=dyn_neckchain_01, swing1LimitDegrees=20, swing2LimitDegrees=10, twistLimitDegrees=1}
 {
  bone1Name=base_neckchain
  bone2Name=spine_05
  swing1LimitDegrees=45
  swing1Motion=Free
  swing2LimitDegrees=45
  swing2Motion=Free
  twistLimitDegrees=45
  twistMotion=Free
 }
 {
  bone1Name=spine_04
  bone2Name=spine_03
  swing1LimitDegrees=45
  swing1Motion=Free
  swing2LimitDegrees=45
  swing2Motion=Free
  twistLimitDegrees=45
  twistMotion=Free
 }
 {
  bone1Name=spine_05
  bone2Name=spine_04
  swing1LimitDegrees=45
  swing1Motion=Free
  swing2LimitDegrees=45
  swing2Motion=Free
  twistLimitDegrees=45
  twistMotion=Free
 }
]
