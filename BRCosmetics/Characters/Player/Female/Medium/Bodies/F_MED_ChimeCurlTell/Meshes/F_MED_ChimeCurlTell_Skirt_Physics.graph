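# Physics asset for the skirt of the ChimeCurlTell medium female player cosmetic body, covering only the right side.
# Single capsule bodies sit on the right thigh, calf, lower calf twist bone and the dyn_skirt_b_r skirt bone.
# The skirt bone is constrained to the thigh with tight swing and twist limits.
# The thigh-to-calf and calf-to-twist constraints leave all angular motion free.
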PhysicsAsset /BRCosmetics/Characters/Player/Female/Medium/Bodies/F_MED_ChimeCurlTell/Meshes/F_MED_ChimeCurlTell_Skirt_Physics
bodies {
 calf_r [{shapeType=Capsule, length=12.5, radius=8, shapeName=calf_r_capsule, center=<15.125 0 0>, rotation=<rot 90 0 0>}]
 calf_twist_02_r [{shapeType=Capsule, length=14, radius=8, shapeName=calf_twist_02_r_capsule, center=<12.6693 -2.8167 0>, rotation=<rot 90 -8.1789 0>}]
 dyn_skirt_b_r [{shapeType=Capsule, length=10.5, radius=3, shapeName=dyn_skirt_b_r_capsule, center=<10.7922 0 0>, rotation=<rot 90 0 0>}]
 thigh_r [{shapeType=Capsule, length=14.3823, radius=9.9506, shapeName=thigh_r_capsule, center=<17.26 -1.9642 0>, rotation=<rot 90 0 0>}]
}
constraints [
 {bone1Name=dyn_skirt_b_r, bone2Name=thigh_r, swing1LimitDegrees=25, swing2LimitDegrees=10, twistLimitDegrees=1}
 {
  bone1Name=calf_r
  bone2Name=thigh_r
  swing1LimitDegrees=45
  swing1Motion=Free
  swing2LimitDegrees=45
  swing2Motion=Free
  twistLimitDegrees=45
  twistMotion=Free
 }
 {
  bone1Name=calf_twist_02_r
  bone2Name=calf_r
  swing1LimitDegrees=45
  swing1Motion=Free
  swing2LimitDegrees=45
  swing2Motion=Free
  twistLimitDegrees=45
  twistMotion=Free
 }
]
