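# Control Rig for the F_MED_ChimeCurlTell face accessory that applies corrective offsets to hair bones according to player state and head pose.
# A branch chain checks emote, crouching, then skydiving: emote and skydiving apply fixed full-weight offsets.
# Crouching (split by targeting) and the default state weight offsets with spherical pose readers tracking the head relative to spine_05.
# Several other pose, yaw-remap and FK control nodes sit disconnected from execution.
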ControlRigBlueprint /BRCosmetics/Characters/Player/Female/Medium/Bodies/F_MED_ChimeCurlTell/Meshes/Parts/F_MED_ChimeCurlTell_FaceAcc_CtrlRig
graphs [
 RigVMFunctionLibrary {links [], nodes []}
 RigVMModel {nodes [
   RigUnit_BeginExecution = &RigUnit_BeginExecution()
   RigVMFunction_Sequence = &RigVMFunction_Sequence(ExecuteContext=@RigUnit_BeginExecution.ExecutePin)
   n1 = &RigVMFunction_ControlFlowBranch(ExecuteContext=@RigVMFunction_Sequence.B, Condition=RigVMVariableNode() {variable=is Playing Emote}.Value)
   RigVMFunction_ControlFlowBranch = &RigVMFunction_ControlFlowBranch(ExecuteContext=@n1.False, Condition=RigVMVariableNode() {variable=is crouching}.Value)
   n2 = &RigVMFunction_ControlFlowBranch(ExecuteContext=@RigVMFunction_ControlFlowBranch.False, Condition=RigVMVariableNode() {variable=is SkyDiving}.Value)
   n3 = &RigUnit_SphericalPoseReader(
    ActiveRegionScaleFactors=(PositiveWidth=3.5,NegativeWidth=0.5,PositiveHeight=0.5,NegativeHeight=3.5)
    ActiveRegionSize=0.1
    Debug=(bDrawDebug=true,bDraw2D=False,bDrawLocalAxes=False,DebugScale=25,DebugSegments=20,DebugThickness=0.25)
    DriverAxis=(X=1,Y=0,Z=0)
    DriverItem=(Type=Bone,Name="head")
    FalloffRegionScaleFactors=(PositiveWidth=1,NegativeWidth=0.7,PositiveHeight=1,NegativeHeight=1)
    FalloffSize=0.35
    FlipHeightScaling=False
    FlipWidthScaling=False
    OptionalParentItem=(Type=Bone,Name="spine_05")
    RotationOffset=(X=90,Y=0,Z=-145)
    ExecutePin=@n2.False
   )
   > &RigUnit_ModifyTransforms(
    ItemToModify="((Item=(Type=Bone,Name=\"base_Hair_Master\"),Transform=(Rotation=(X=-0.0186,Y=-0.0001,Z=-0.0062,W=0.9998),Translation=(X=2.9259,Y=-2.8265,Z=0),Scale3D=(X=1,Y=1,Z=1))),(Item=(Type=Bone,Name=\"base_hair_21_r\"),Transform=(Rotation=(X=0,Y=0.1546,Z=0,W=-0.988),Tra ...(truncated)"
    WeightMinimum=0
    Weight=@n3.OutputParam
   )
   > n4 = &RigUnit_SphericalPoseReader(
    ActiveRegionScaleFactors=(PositiveWidth=1,NegativeWidth=2,PositiveHeight=1,NegativeHeight=1)
    ActiveRegionSize=0.1
    Debug=(bDrawDebug=true,bDraw2D=False,bDrawLocalAxes=False,DebugScale=25,DebugSegments=20,DebugThickness=0.25)
    DriverAxis=(X=1,Y=0,Z=0)
    DriverItem=(Type=Bone,Name="head")
    FalloffRegionScaleFactors=(PositiveWidth=0.1,NegativeWidth=3,PositiveHeight=1.2,NegativeHeight=1)
    FalloffSize=0.3
    FlipHeightScaling=False
    FlipWidthScaling=False
    OptionalParentItem=(Type=Bone,Name="spine_05")
    RotationOffset=(X=28.5019,Y=-25.1638,Z=-85.549)
   )
   > &RigUnit_ModifyTransforms(
    ItemToModify="((Item=(Type=Bone,Name=\"base_hair_21_r\"),Transform=(Rotation=(X=0.4698,Y=-0.3624,Z=0.2825,W=0.7537),Translation=(X=4.2831,Y=-1.1561,Z=0.4988),Scale3D=(X=1,Y=1,Z=1))),(Item=(Type=Bone,Name=\"base_hair_17_l\"),Transform=(Rotation=(X=0,Y=0,Z=0,W=1),Translation= ...(truncated)"
    WeightMinimum=0
    Weight=@n4.OutputParam
   )
   &RigUnit_ModifyTransforms(
    ItemToModify=((Item=(Type=Bone,Name="base_Hair_Master"),Transform=(Rotation=(X=0,Y=0,Z=0,W=-1),Translation=(X=1.9733,Y=2.3225,Z=0),Scale3D=(X=1,Y=1,Z=1))))
    Weight=1
    WeightMinimum=0
    ExecutePin=@n2.True
   )
   n5 = &RigVMFunction_ControlFlowBranch(ExecuteContext=@RigVMFunction_ControlFlowBranch.True, Condition=RigVMVariableNode() {variable=is targeting}.Value)
   n6 = &RigUnit_SphericalPoseReader(
    ActiveRegionScaleFactors=(PositiveWidth=3.5,NegativeWidth=0.5,PositiveHeight=0.5,NegativeHeight=3.5)
    ActiveRegionSize=0.1
    Debug=(bDrawDebug=true,bDraw2D=False,bDrawLocalAxes=False,DebugScale=25,DebugSegments=20,DebugThickness=0.25)
    DriverAxis=(X=1,Y=0,Z=0)
    DriverItem=(Type=Bone,Name="head")
    FalloffRegionScaleFactors=(PositiveWidth=1,NegativeWidth=0.7,PositiveHeight=1,NegativeHeight=1)
    FalloffSize=0.35
    FlipHeightScaling=False
    FlipWidthScaling=False
    OptionalParentItem=(Type=Bone,Name="spine_05")
    RotationOffset=(X=90,Y=0,Z=-148)
    ExecutePin=@n5.False
   )
   > &RigUnit_ModifyTransforms(
    ItemToModify="((Item=(Type=Bone,Name=\"base_Hair_Master\"),Transform=(Rotation=(X=-0.1473,Y=-0.0366,Z=-0.2381,W=0.9593),Translation=(X=3.6147,Y=1.3655,Z=-0.4572),Scale3D=(X=1,Y=1,Z=1))),(Item=(Type=Bone,Name=\"base_hair_21_r\"),Transform=(Rotation=(X=0,Y=0.0813,Z=0,W=0.9967 ...(truncated)"
    WeightMinimum=0
    Weight=@n6.OutputParam
   )
   n7 = &RigUnit_SphericalPoseReader(
    ActiveRegionScaleFactors=(PositiveWidth=0.8,NegativeWidth=0.8,PositiveHeight=0.8,NegativeHeight=0.8)
    ActiveRegionSize=0.08
    Debug=(bDrawDebug=true,bDraw2D=False,bDrawLocalAxes=False,DebugScale=25,DebugSegments=20,DebugThickness=0.25)
    DriverAxis=(X=1,Y=0,Z=0)
    DriverItem=(Type=Bone,Name="head")
    FalloffRegionScaleFactors=(PositiveWidth=0.7,NegativeWidth=0.7,PositiveHeight=0.7,NegativeHeight=0.7)
    FalloffSize=0.5
    FlipHeightScaling=False
    FlipWidthScaling=False
    OptionalParentItem=(Type=Bone,Name="spine_05")
    RotationOffset=(X=-45,Y=0,Z=40)
    ExecutePin=@n5.True
   )
   > &RigUnit_ModifyTransforms(
    ItemToModify="((Item=(Type=Bone,Name=\"base_hair_21_r\"),Transform=(Rotation=(X=0,Y=0,Z=0,W=-1),Translation=(X=0,Y=0,Z=-0.2364),Scale3D=(X=1,Y=1,Z=1))),(Item=(Type=Bone,Name=\"base_Hair_clipping_joint\"),Transform=(Rotation=(X=-0.0636,Y=-0.0453,Z=-0.4539,W=0.8876),Translati ...(truncated)"
    WeightMinimum=0
    Weight=@n7.OutputParam
   )
   &RigUnit_ModifyTransforms(
    ItemToModify="((Item=(Type=Bone,Name=\"base_Hair_Master\"),Transform=(Rotation=(X=0,Y=0,Z=-0.0855,W=0.9963),Translation=(X=3,Y=0,Z=0),Scale3D=(X=1,Y=1,Z=1))),(Item=(Type=Bone,Name=\"base_Hair_R_01\"),Transform=(Rotation=(X=-0.0461,Y=0.0521,Z=-0.1673,W=-0.9835),Translation=( ...(truncated)"
    Weight=1
    WeightMinimum=0
    ExecutePin=@n1.True
   )
   n8 = &RigVMFunction_ControlFlowBranch(ExecuteContext=(), Condition=RigVMVariableNode() {variable=is targeting}.Value)
   n9 = &RigUnit_SphericalPoseReader(
    ActiveRegionScaleFactors=(PositiveWidth=0.8,NegativeWidth=0.8,PositiveHeight=0.8,NegativeHeight=0.8)
    ActiveRegionSize=0.07
    Debug=(bDrawDebug=true,bDraw2D=False,bDrawLocalAxes=False,DebugScale=25,DebugSegments=20,DebugThickness=0.25)
    DriverAxis=(X=1,Y=0,Z=0)
    DriverItem=(Type=Bone,Name="head")
    FalloffRegionScaleFactors=(PositiveWidth=0.7,NegativeWidth=0.7,PositiveHeight=0.7,NegativeHeight=0.7)
    FalloffSize=0.2
    FlipHeightScaling=False
    FlipWidthScaling=False
    OptionalParentItem=(Type=Bone,Name="spine_05")
    RotationOffset=(X=-45,Y=0,Z=40)
    ExecutePin=@n8.True
   )
   > &RigUnit_ModifyTransforms(
    ItemToModify="((Item=(Type=Bone,Name=\"base_hair_21_r\"),Transform=(Rotation=(X=-0.7187,Y=0.3245,Z=0.3657,W=0.4944),Translation=(X=0.0001,Y=0,Z=0),Scale3D=(X=1,Y=1,Z=1))),(Item=(Type=Bone,Name=\"base_Hair_R_00\"),Transform=(Rotation=(X=0,Y=0,Z=0,W=1),Translation=(X=0,Y=0,Z= ...(truncated)"
    WeightMinimum=0
    Weight=@n9.OutputParam
   )
   n10 = &RigUnit_SphericalPoseReader(
    ActiveRegionScaleFactors=(PositiveWidth=0.5,NegativeWidth=0.5,PositiveHeight=0.5,NegativeHeight=0.5)
    ActiveRegionSize=0.1
    Debug=(bDrawDebug=true,bDraw2D=False,bDrawLocalAxes=False,DebugScale=25,DebugSegments=20,DebugThickness=0.25)
    DriverAxis=(X=1,Y=0,Z=0)
    DriverItem=(Type=Bone,Name="head")
    ExecutePin=()
    FalloffRegionScaleFactors=(PositiveWidth=0.7,NegativeWidth=0.7,PositiveHeight=0.7,NegativeHeight=0.7)
    FalloffSize=0.2
    FlipHeightScaling=False
    FlipWidthScaling=False
    OptionalParentItem=(Type=Bone,Name="spine_05")
    RotationOffset=(X=-60,Y=0,Z=15)
   )
   > &RigUnit_ModifyTransforms(
    ItemToModify="((Item=(Type=Bone,Name=\"base_hair_21_r\"),Transform=(Rotation=(X=0,Y=0.2862,Z=0,W=0.9582),Translation=(X=0,Y=0,Z=0),Scale3D=(X=1,Y=1,Z=1))),(Item=(Type=Bone,Name=\"base_hair_17_l\"),Transform=(Rotation=(X=-0.087,Y=-0.4164,Z=-0.185,W=0.8859),Translation=(X=0,Y ...(truncated)"
    WeightMinimum=0
    Weight=@n10.OutputParam
   )
   &RigUnit_ModifyTransforms(
    ExecutePin=()
    ItemToModify="((Item=(Type=Bone,Name=\"base_hair_21_r\"),Transform=(Rotation=(X=-0.3211,Y=0.2455,Z=0.3671,W=0.8378),Translation=(X=0,Y=0,Z=0),Scale3D=(X=1,Y=1,Z=1))),(Item=(Type=Bone,Name=\"base_Hair_R_00\"),Transform=(Rotation=(X=0.0857,Y=0.1684,Z=-0.2055,W=-0.9602),Transl ...(truncated)"
    WeightMinimum=0
    Weight=@n9.OutputParam
   )
   &RigUnit_ModifyTransforms(
    ExecutePin=()
    ItemToModify="((Item=(Type=Bone,Name=\"base_Hair_Master\"),Transform=(Rotation=(X=-0.1515,Y=-0.0075,Z=-0.0488,W=0.9872),Translation=(X=3.4637,Y=-0.0884,Z=0.0001),Scale3D=(X=1,Y=1,Z=1))),(Item=(Type=Bone,Name=\"base_hair_21_r\"),Transform=(Rotation=(X=0,Y=0.0813,Z=0,W=0.9967 ...(truncated)"
    Weight=1
    WeightMinimum=0
   )
   n11 = &RigVMFunction_MathTransformMakeRelative(
    Global=&RigUnit_GetRelativeTransformForItem(Child=(Type=Bone,Name="head"), Parent=(Type=Bone,Name="head"), bChildInitial=False, bParentInitial=true).RelativeTransform
    Parent=&RigUnit_GetRelativeTransformForItem(Child=(Type=Bone,Name="spine_05"), Parent=(Type=Bone,Name="spine_05"), bChildInitial=False, bParentInitial=true).RelativeTransform
   )
   &RigVMFunction_MathDoubleRemap(
    SourceMaximum=-55
    SourceMinimum=-1.4
    TargetMaximum=1
    TargetMinimum=0
    Value=RigVMDispatchNode(Struct=/Script/RigVM.RigVMFunction_MathQuaternionToRotator(Value=@n11.Local.Rotation) {is=struct RigVMUnitNode, template="Cast::Execute(in Value,out Result)"}.Result) {template="DISPATCH_RigVMDispatch_BreakStruct(in Struct,out Elements)"}.Elements.Yaw
   )
   setFKCtrl = RigVMFunctionReferenceNode(ExecuteContext=(), setFK_Ctrl=false)
  ]}
 RigVMModel Construction Event Graph {nodes [
   n1 = &RigUnit_PrepareForExecution()
   &RigUnit_HierarchyImportFromSkeleton(NameSpace=None, bIncludeCurves=False, bIncludeMeshSockets=False, bIncludeVirtualBones=True, ExecutePin=@n1.ExecutePin)
   addFKCtrl = RigVMFunctionReferenceNode(
    Custom_FKCtrl=()
    ExecuteContext=()
    Options_FKCtrl="(pelvis_1_#1=false,spine_3_#2=false,clavicle_5_#3=true,arm_7_#4=false,leg_9_#5=false,head_11_46DC15EF465CD ...(truncated)"
    Shape_Settings="(pelvis_2_#6=(bVisible=True,Name=\"Hexagon_Thin\",Color=(R=0.0117,G=0.8281,B=1,A=1),Transform=(Rotation=(X=0,Y=-0.7071,Z=0,W=0.7071),Translation=(X=2,Y=0,Z=0),Scale3D=(X=5,Y=5,Z=7))),spine_4_#7=(bVi ...(truncated)"
   )
  ]}
]
variables [
 NewVar {type=double, default=0}
 is Playing Emote {type=bool, default=False}
 is SkyDiving {type=bool, default=False}
 is crouch walking {type=bool, default=False}
 is crouching {type=bool, default=False}
 is targeting {type=bool, default=False}
 is walking2d {type=bool, default=False}
]
