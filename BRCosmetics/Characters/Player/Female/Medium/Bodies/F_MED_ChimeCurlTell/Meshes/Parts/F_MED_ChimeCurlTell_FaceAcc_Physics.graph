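# Physics asset for the FaceAcc part of the medium female player cosmetic ChimeCurlTell, set up for dynamic hair.
# Capsules on spine, clavicles, neck, head and left upper arm form the body volume around the hair.
# Left, middle and right chains root through base_Hair_Master to neck_02, a clipping joint branches off the right base, and two side locks root on the head.
# Dyn joints use limited swing and tight or locked twist, while body and base links rotate freely.
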PhysicsAsset /BRCosmetics/Characters/Player/Female/Medium/Bodies/F_MED_ChimeCurlTell/Meshes/Parts/F_MED_ChimeCurlTell_FaceAcc_Physics
bodies {
 base_Hair_L_01 [{shapeType=Capsule, length=0.505, radius=0.505, shapeName=base_Hair_L_01_capsule, rotation=<rot 0 -90 0>}]
 base_Hair_M_01 [{shapeType=Capsule, length=0.505, radius=0.505, shapeName=base_Hair_M_01_capsule, rotation=<rot 0 -90 0>}]
 base_Hair_Master [{shapeType=Capsule, length=0.505, radius=0.505, shapeName=base_Hair_Master_capsule, rotation=<rot 0 -90 0>}]
 base_Hair_R_01 [{shapeType=Capsule, length=0.505, radius=0.505, shapeName=base_Hair_R_01_capsule, rotation=<rot 0 -90 0>}]
 base_Hair_clipping_joint [{shapeType=Capsule, length=0.505, radius=0.505, shapeName=base_Hair_clipping_joint_capsule, rotation=<rot 0 -90 0>}]
 base_hair_17_l [{shapeType=Capsule, length=2.2863, radius=2, shapeName=base_hair_17_l_capsule, center=<2.8448 0 0>, rotation=<rot 90 0 0>}]
 base_hair_21_r [{shapeType=Capsule, length=3, radius=1, shapeName=base_hair_21_r_capsule, rotation=<rot 90 0 0>}]
 clavicle_l [
  {
   shapeType=Capsule
   length=6.5
   radius=6.1431
   shapeName=clavicle_l_capsule
   center=<4.6332 -1.148 -2.0722>
   rotation=<rot -0.0166 -87.6925 90.2584>
  }
  {shapeType=Capsule, length=4, radius=2, shapeName=clavicle_l_capsule, center=<1.2598 -4.2051 1.6319>, rotation=<rot 80.3498 0.4617 -2.3919>}
  {
   shapeType=Capsule
   length=5
   radius=3
   shapeName=clavicle_l_capsule
   center=<2.1042 -2.097 2.3337>
   rotation=<rot -67.5872 179.9994 -179.9993>
  }
 ]
 clavicle_r [
  {
   shapeType=Capsule
   length=7.5
   radius=5.505
   shapeName=clavicle_r_capsule
   center=<-5.747 1.2118 1.3875>
   rotation=<rot 0.8707 -89.9671 97.3367>
  }
  {
   shapeType=Capsule
   length=8
   radius=5
   shapeName=clavicle_r_capsule
   center=<-2.2393 2.3343 0.5355>
   rotation=<rot -88.5099 179.9931 -173.7606>
  }
 ]
 dyn_Hair_L_01 [{shapeType=Capsule, length=5, radius=2, shapeName=dyn_Hair_L_01_capsule, center=<5 0.856 0>, rotation=<rot 90 -0.084 0>}]
 dyn_Hair_L_02 [
  {
   shapeType=Capsule
   length=5
   radius=2
   shapeName=dyn_Hair_L_02_capsule
   center=<6.1231 -0.1752 -1.8737>
   rotation=<rot 27.7525 179.9997 177.7877>
  }
  {
   shapeType=Capsule
   length=4
   radius=2
   shapeName=dyn_Hair_L_02_capsule
   center=<2.3684 0.2693 -0.573>
   rotation=<rot -24.9124 -10.7304 24.2214>
  }
 ]
 dyn_Hair_M_01 [{shapeType=Capsule, length=3.6195, radius=2.5542, shapeName=dyn_Hair_M_01_capsule, center=<4 0.368 0>, rotation=<rot 90 -0.084 0>}]
 dyn_Hair_M_02 [
  {shapeType=Capsule, length=1.2053, radius=2.7848, shapeName=dyn_Hair_M_02_capsule, center=<6.03 -1.4163 0>}
  {shapeType=Capsule, length=0.9758, radius=2.5542, shapeName=dyn_Hair_M_02_capsule, center=<1.551 1.5156 -0.2461>, rotation=<rot 0 0 0.0839>}
 ]
 dyn_Hair_R_01 [{shapeType=Capsule, length=2.8543, radius=2.5, shapeName=dyn_Hair_R_01_capsule, center=<3.9625 0.3893 0>, rotation=<rot 90 -0.084 0>}]
 dyn_Hair_R_02 [
  {
   shapeType=Capsule
   length=4.958
   radius=2
   shapeName=dyn_Hair_R_02_capsule
   center=<6.2242 0.1373 0.9967>
   rotation=<rot 23.5306 -0.0003 -13.6573>
  }
  {
   shapeType=Capsule
   length=4.0105
   radius=1.9895
   shapeName=dyn_Hair_R_02_capsule
   center=<2.3852 0.2314 -0.2704>
   rotation=<rot -20.0946 175.6367 -158.9366>
  }
 ]
 dyn_Hair_clipping_joint [{
   shapeType=Capsule
   length=4.626
   radius=1.6362
   shapeName=dyn_Hair_clipping_joint_capsule
   center=<0.6127 1.7932 0.4007>
   rotation=<rot 0 127.4303 -65.8209>
  }]
 dyn_hair_17_l [{shapeType=Capsule, length=4.4656, radius=2, shapeName=dyn_hair_17_l_capsule, center=<3.8467 -0.4458 0>, rotation=<rot 90 0 0>}]
 dyn_hair_17_l_02 [{
   shapeType=Capsule
   length=2.9764
   radius=1.9397
   shapeName=dyn_hair_17_l_02_capsule
   center=<1.5144 -0.387 0.1425>
   rotation=<rot 67.4134 -0.0035 -37.1974>
  }]
 dyn_hair_21_00_r [{shapeType=Capsule, length=3, radius=2, shapeName=dyn_hair_21_00_r_capsule, center=<3.917 0 0>, rotation=<rot 90 0 0>}]
 dyn_hair_21_01_r [{shapeType=Capsule, length=4, radius=2.2, shapeName=dyn_hair_21_01_r_capsule, center=<3 1.5 0>, rotation=<rot 90 0 180>}]
 head [{
   shapeType=Capsule
   length=0.0421
   radius=7.674
   shapeName=head_capsule
   center=<5.7936 0.7436 -0.5646>
   rotation=<rot -2.7711 -89.84 -0.3579>
  }]
 neck_02 [
  {
   shapeType=Capsule
   length=5.494
   radius=5.5193
   shapeName=neck_02_capsule
   center=<-6.4873 3.7539 -0.0169>
   rotation=<rot 0.6516 -104.8694 87.6477>
  }
  {shapeType=Capsule, length=6, radius=4.9464, shapeName=neck_02_capsule, center=<5.9879 -0.0586 0>, rotation=<rot 90 145.981 0>}
  {shapeType=Capsule, length=5, radius=5, shapeName=neck_02_capsule, center=<-7.8304 -0.376 -2.1923>, rotation=<rot 90 -11.318 0>}
  {shapeType=Capsule, length=5, radius=5, shapeName=neck_02_capsule, center=<-7.4235 -0.6668 1.5538>, rotation=<rot 87.6332 -24.4862 -19.7557>}
 ]
 spine_04 [
  {shapeType=Capsule, length=10.005, radius=8.005, shapeName=spine_04_capsule, center=<6.78 -0.1514 0>, rotation=<rot 0 -90 0>}
  {shapeType=Capsule, length=15, radius=8, shapeName=spine_04_capsule, center=<2.2186 -1.0369 0>}
 ]
 spine_05 [{shapeType=Capsule, length=11.6872, radius=7.3005, shapeName=spine_05_capsule, center=<2.1164 -0.7325 0>, rotation=<rot 0 -77.7594 0>}]
 upperarm_l [{
   shapeType=Capsule
   length=8.5641
   radius=4.5
   shapeName=upperarm_l_capsule
   center=<4.4798 -0.6787 -0.94>
   rotation=<rot 84.3964 -0.1568 -0.1727>
  }]
}
constraints [
 {
  bone1Name=clavicle_l
  bone2Name=spine_05
  swing1LimitDegrees=45
  swing1Motion=Free
  swing2LimitDegrees=45
  swing2Motion=Free
  twistLimitDegrees=45
  twistMotion=Free
 }
 {
  bone1Name=clavicle_r
  bone2Name=spine_05
  swing1LimitDegrees=45
  swing1Motion=Free
  swing2LimitDegrees=45
  swing2Motion=Free
  twistLimitDegrees=45
  twistMotion=Free
 }
 {
  bone1Name=head
  bone2Name=neck_02
  swing1LimitDegrees=45
  swing1Motion=Free
  swing2LimitDegrees=45
  swing2Motion=Free
  twistLimitDegrees=45
  twistMotion=Free
 }
 {bone1Name=dyn_Hair_L_01, bone2Name=base_Hair_L_01, swing1LimitDegrees=35, swing2LimitDegrees=12, twistLimitDegrees=1}
 {bone1Name=dyn_Hair_L_02, bone2Name=dyn_Hair_L_01, swing1LimitDegrees=55, swing2LimitDegrees=10, twistLimitDegrees=1}
 {bone1Name=dyn_Hair_M_01, bone2Name=base_Hair_M_01, swing1LimitDegrees=45, swing2LimitDegrees=7, twistLimitDegrees=1, twistMotion=Locked}
 {bone1Name=dyn_Hair_R_01, bone2Name=base_Hair_R_01, swing1LimitDegrees=35, swing2LimitDegrees=7, twistLimitDegrees=1}
 {bone1Name=dyn_Hair_R_02, bone2Name=dyn_Hair_R_01, swing1LimitDegrees=55, swing2LimitDegrees=10, twistLimitDegrees=1}
 {
  bone1Name=base_hair_21_r
  bone2Name=head
  swing1LimitDegrees=45
  swing1Motion=Free
  swing2LimitDegrees=45
  swing2Motion=Free
  twistLimitDegrees=45
  twistMotion=Free
 }
 {bone1Name=dyn_hair_17_l, bone2Name=base_hair_17_l, swing1LimitDegrees=10, swing2LimitDegrees=35, twistLimitDegrees=1}
 {
  bone1Name=base_hair_17_l
  bone2Name=head
  swing1LimitDegrees=45
  swing1Motion=Free
  swing2LimitDegrees=45
  swing2Motion=Free
  twistLimitDegrees=45
  twistMotion=Free
 }
 {
  bone1Name=neck_02
  bone2Name=spine_05
  swing1LimitDegrees=45
  swing1Motion=Free
  swing2LimitDegrees=45
  swing2Motion=Free
  twistLimitDegrees=45
  twistMotion=Free
 }
 {bone1Name=dyn_hair_21_00_r, bone2Name=base_hair_21_r, swing1LimitDegrees=10, swing2LimitDegrees=15, twistLimitDegrees=5}
 {bone1Name=dyn_hair_21_01_r, bone2Name=dyn_hair_21_00_r, swing1LimitDegrees=10, swing2LimitDegrees=25, twistLimitDegrees=5}
 {bone1Name=dyn_hair_17_l_02, bone2Name=dyn_hair_17_l, swing1LimitDegrees=3, swing2LimitDegrees=35, twistLimitDegrees=1}
 {
  bone1Name=spine_05
  bone2Name=spine_04
  swing1LimitDegrees=45
  swing1Motion=Free
  swing2LimitDegrees=45
  swing2Motion=Free
  twistLimitDegrees=45
  twistMotion=Free
 }
 {
  bone1Name=base_Hair_Master
  bone2Name=neck_02
  swing1LimitDegrees=45
  swing1Motion=Free
  swing2LimitDegrees=45
  swing2Motion=Free
  twistLimitDegrees=45
  twistMotion=Free
 }
 {
  bone1Name=base_Hair_L_01
  bone2Name=base_Hair_Master
  swing1LimitDegrees=45
  swing1Motion=Free
  swing2LimitDegrees=45
  swing2Motion=Free
  twistLimitDegrees=45
  twistMotion=Free
 }
 {
  bone1Name=base_Hair_M_01
  bone2Name=base_Hair_Master
  swing1LimitDegrees=45
  swing1Motion=Free
  swing2LimitDegrees=45
  swing2Motion=Free
  twistLimitDegrees=45
  twistMotion=Free
 }
 {
  bone1Name=base_Hair_R_01
  bone2Name=base_Hair_Master
  swing1LimitDegrees=45
  swing1Motion=Free
  swing2LimitDegrees=45
  swing2Motion=Free
  twistLimitDegrees=45
  twistMotion=Free
 }
 {
  bone1Name=base_Hair_clipping_joint
  bone2Name=base_Hair_R_01
  swing1LimitDegrees=45
  swing1Motion=Free
  swing2LimitDegrees=45
  swing2Motion=Free
  twistLimitDegrees=45
  twistMotion=Free
 }
 {bone1Name=dyn_Hair_clipping_joint, bone2Name=base_Hair_clipping_joint, swing1LimitDegrees=45, swing2LimitDegrees=10, twistLimitDegrees=1}
 {bone1Name=dyn_Hair_M_02, bone2Name=dyn_Hair_M_01, swing1LimitDegrees=55, swing2LimitDegrees=10, twistLimitDegrees=1}
 {
  bone1Name=upperarm_l
  bone2Name=clavicle_l
  swing1LimitDegrees=45
  swing1Motion=Free
  swing2LimitDegrees=45
  swing2Motion=Free
  twistLimitDegrees=45
  twistMotion=Free
 }
]
