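# Post-process AnimBP for the F_MED_ChimeCurlTell FaceAcc mesh, animating the character's hair and earrings.
# The parent mesh pose feeds a state-driven Control Rig and RigidBody simulation, then is cached.
# When a backbling or cape is worn, it blends to bone constraints and offsets that repose hair against neck, spine and shoulders.
# AnimDynamics springs and minor corrections finish the pose; the EventGraph updates state flags and gravity override.
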AnimBlueprint /BRCosmetics/Characters/Player/Female/Medium/Bodies/F_MED_ChimeCurlTell/Meshes/Parts/F_MED_ChimeCurlTell_FaceAcc_AnimBP
target_skeleton=&/F_MED_ChimeCurlTell_FaceAcc_Skeleton
graphs [
 AnimGraph {nodes [
   "Constraint - Base_earing_r (R)"(Alpha=1, ConstraintWeights_0=1) {boneToModify {boneName=Base_earing_r}, constraintSetup [{transformType=Rotation, targetBone {boneName=head}}], constraintWeights [1]}
   n1 = "Constraint - dyn_hair_21_00_r (PPP)"!(ComponentPose, Alpha=0.2, ConstraintWeights_0=0, ConstraintWeights_1:float=1, ConstraintWeights_2:float=0) -> (Pose) {
    alphaScaleBiasClamp {bClampResult=true, clampMax=0.2}
    boneToModify {boneName=dyn_hair_21_00_r}
    constraintSetup [{targetBone {boneName=clavicle_l}}, {targetBone {boneName=neck_02}}, {targetBone {boneName=upperarm_l}}]
    constraintWeights [1,1,1]
   }
   "Constraint - dyn_hair_21_01_r (PPP)"!(ComponentPose=@n1, Alpha=0.2, ConstraintWeights_0=0.5, ConstraintWeights_1:float=0.5, ConstraintWeights_2:float=0) -> (Pose) {
    alphaScaleBiasClamp {bClampResult=true, clampMax=0.2}
    boneToModify {boneName=dyn_hair_21_01_r}
    constraintSetup [{targetBone {boneName=clavicle_r}}, {targetBone {boneName=neck_02}}, {targetBone {boneName=upperarm_l}}]
    constraintWeights [1,1,1]
   }
   n2 = "Constraint - dyn_hair_short_bk_r_02 (PP)"!(ComponentPose, Alpha=1, ConstraintWeights_0=1, ConstraintWeights_1:float=0) -> (Pose) {
    boneToModify {boneName=dyn_hair_short_bk_r_02}
    constraintSetup [{targetBone {boneName=clavicle_r}}, {targetBone {boneName=neck_02}}]
    constraintWeights [1,1]
   }
   n3 = "Constraint - dyn_hair_short_bk_r_03 (PPP)"!(ComponentPose=@n2, Alpha=1, ConstraintWeights_0=0, ConstraintWeights_1:float=0, ConstraintWeights_2:float=1) -> (Pose) {
    boneToModify {boneName=dyn_hair_short_bk_r_03}
    constraintSetup [{targetBone {boneName=neck_01}}, {targetBone {boneName=spine_05}}, {targetBone {boneName=clavicle_r}}]
    constraintWeights [1,1,1]
   }
   n4 = "Constraint - dyn_hair_short_bk_l_01 (PP)"!(ComponentPose=@n3, Alpha=1, ConstraintWeights_0=0.5, ConstraintWeights_1:float=0.5) -> (Pose) {
    boneToModify {boneName=dyn_hair_short_bk_l_01}
    constraintSetup [{targetBone {boneName=head}}, {targetBone {boneName=neck_02}}]
    constraintWeights [1,1]
   }
   n5 = "Constraint - dyn_hair_short_bk_l_02 (PP)"!(ComponentPose=@n4, Alpha=1, ConstraintWeights_0=1, ConstraintWeights_1:float=0) -> (Pose) {
    boneToModify {boneName=dyn_hair_short_bk_l_02}
    constraintSetup [{targetBone {boneName=clavicle_l}}, {targetBone {boneName=neck_02}}]
    constraintWeights [1,1]
   }
   n6 = "Constraint - dyn_hair_short_bk_l_03 (PPP)"!(ComponentPose=@n5, Alpha=1, ConstraintWeights_0=0, ConstraintWeights_1:float=0, ConstraintWeights_2:float=1) -> (Pose) {
    boneToModify {boneName=dyn_hair_short_bk_l_03}
    constraintSetup [{targetBone {boneName=neck_01}}, {targetBone {boneName=spine_05}}, {targetBone {boneName=clavicle_l}}]
    constraintWeights [1,1,1]
   }
   n7 = "Constraint - dyn_hair_short_bk_c_01 (PP)"!(ComponentPose=@n6, Alpha=1, ConstraintWeights_0=0.5, ConstraintWeights_1:float=0.5) -> (Pose) {
    boneToModify {boneName=dyn_hair_short_bk_c_01}
    constraintSetup [{targetBone {boneName=head}}, {targetBone {boneName=neck_02}}]
    constraintWeights [1,1]
   }
   n8 = "Constraint - dyn_hair_short_bk_c_02 (PP)"!(ComponentPose=@n7, Alpha=1, ConstraintWeights_0=0.5, ConstraintWeights_1:float=0.5) -> (Pose) {
    boneToModify {boneName=dyn_hair_short_bk_c_02}
    constraintSetup [{targetBone {boneName=neck_01}}, {targetBone {boneName=neck_02}}]
    constraintWeights [1,1]
   }
   n9 = "Constraint - dyn_hair_short_bk_c_03 (PP)"!(ComponentPose=@n8, Alpha=1, ConstraintWeights_0=0.5, ConstraintWeights_1:float=0.5) -> (Pose) {
    boneToModify {boneName=dyn_hair_short_bk_c_03}
    constraintSetup [{targetBone {boneName=neck_01}}, {targetBone {boneName=spine_05}}]
    constraintWeights [1,1]
   }
   n10 = "Constraint - dyn_hair_short_side_r_01 (PP)"!(ComponentPose=@n9, Alpha=1, ConstraintWeights_0=1, ConstraintWeights_1:float=0) -> (Pose) {
    boneToModify {boneName=dyn_hair_short_side_r_01}
    constraintSetup [{targetBone {boneName=head}}, {targetBone {boneName=neck_02}}]
    constraintWeights [1,1]
   }
   n11 = "Constraint - dyn_hair_short_side_r_02 (PP)"!(ComponentPose=@n10, Alpha=1, ConstraintWeights_0=0, ConstraintWeights_1:float=1) -> (Pose) {
    boneToModify {boneName=dyn_hair_short_side_r_02}
    constraintSetup [{targetBone {boneName=clavicle_r}}, {targetBone {boneName=neck_02}}]
    constraintWeights [1,1]
   }
   n12 = "Constraint - dyn_hair_short_side_r_03 (PPP)"!(ComponentPose=@n11, Alpha=1, ConstraintWeights_0=0, ConstraintWeights_1:float=0, ConstraintWeights_2:float=1) -> (Pose) {
    boneToModify {boneName=dyn_hair_short_side_r_03}
    constraintSetup [{targetBone {boneName=neck_01}}, {targetBone {boneName=spine_05}}, {targetBone {boneName=clavicle_r}}]
    constraintWeights [1,1,1]
   }
   n13 = "Constraint - dyn_hair_short_side_l_01 (PP)"!(ComponentPose=@n12, Alpha=1, ConstraintWeights_0=1, ConstraintWeights_1:float=0) -> (Pose) {
    boneToModify {boneName=dyn_hair_short_side_l_01}
    constraintSetup [{targetBone {boneName=head}}, {targetBone {boneName=neck_02}}]
    constraintWeights [1,1]
   }
   n14 = "Constraint - dyn_hair_short_side_l_02 (PP)"!(ComponentPose=@n13, Alpha=1, ConstraintWeights_0=0, ConstraintWeights_1:float=1) -> (Pose) {
    boneToModify {boneName=dyn_hair_short_side_l_02}
    constraintSetup [{targetBone {boneName=clavicle_l}}, {targetBone {boneName=neck_02}}]
    constraintWeights [1,1]
   }
   "Constraint - dyn_hair_short_side_l_03 (PPP)"!(ComponentPose=@n14, Alpha=1, ConstraintWeights_0=0, ConstraintWeights_1:float=0, ConstraintWeights_2:float=1) -> (Pose) {
    boneToModify {boneName=dyn_hair_short_side_l_03}
    constraintSetup [{targetBone {boneName=neck_01}}, {targetBone {boneName=spine_05}}, {targetBone {boneName=clavicle_l}}]
    constraintWeights [1,1,1]
   }
   "Constraint - dyn_hair_short_bk_r_02 (PP)"(Alpha=1, ConstraintWeights_0=1, ConstraintWeights_1=1) {
    boneToModify {boneName=dyn_hair_short_bk_r_02}
    constraintSetup [{targetBone {boneName=clavicle_r}}, {targetBone {boneName=neck_02}}]
    constraintWeights [1,1]
   }
   "Constraint - dyn_Hair_R_02 (PPP)"!(ComponentPose, Alpha=1, ConstraintWeights_0=0.5, ConstraintWeights_1:float=0.2, ConstraintWeights_2:float=0.3) -> (Pose) {
    boneToModify {boneName=dyn_Hair_R_02}
    constraintSetup [{targetBone {boneName=clavicle_r}}, {targetBone {boneName=neck_01}}, {targetBone {boneName=upperarm_r}}]
    constraintWeights [1,1,1]
   }
   Local To Component()
   Local To Component()
   Local To Component()
   "Transform (Modify) Bone - Bone: Base_earing_r"!(ComponentPose, bAlphaBoolEnabled:bool=True, Translation="0,0,0", Rotation="0,-5,0", Scale="1,1,1") -> (Pose) {
    alphaInputType=Bool
    rotationMode=BMM_Additive
    rotationSpace=BCS_BoneSpace
    scaleMode=BMM_Additive
    scaleSpace=BCS_BoneSpace
    translationMode=BMM_Additive
    translationSpace=BCS_BoneSpace
    alphaBoolBlend {blendInTime=0.2, blendOutTime=0.2}
    boneToModify {boneName=Base_earing_r}
    rotation {yaw=-5}
   }
   n15 = "Constraint - dyn_Hair_L_01 (PP)"!(
    ComponentPose=Local To Component(LocalPose=Use cached pose 'Control Rig Pose'())
    Alpha=1
    ConstraintWeights_0=0.5
    ConstraintWeights_1:float=0.5
   ) -> (Pose) {
    boneToModify {boneName=dyn_Hair_L_01}
    constraintSetup [{targetBone {boneName=head}}, {targetBone {boneName=neck_02}}]
    constraintWeights [1,1]
   }
   n16 = "Constraint - dyn_Hair_R_01 (PP)"!(ComponentPose=@n15, Alpha=1, ConstraintWeights_0=0.5, ConstraintWeights_1:float=0.5) -> (Pose) {
    boneToModify {boneName=dyn_Hair_R_01}
    constraintSetup [{targetBone {boneName=neck_01}}, {targetBone {boneName=neck_02}}]
    constraintWeights [1,1]
   }
   n17 = "Constraint - dyn_Hair_M_01 (PP)"!(ComponentPose=@n16, Alpha=1, ConstraintWeights_0=0, ConstraintWeights_1:float=1) -> (Pose) {
    boneToModify {boneName=dyn_Hair_M_01}
    constraintSetup [{targetBone {boneName=neck_01}}, {targetBone {boneName=neck_02}}]
    constraintWeights [1,1]
   }
   n18 = "Constraint - dyn_hair_17_l (PPP)"!(ComponentPose=@n17, Alpha=1, ConstraintWeights_0=0, ConstraintWeights_1:float=1, ConstraintWeights_2:float=0) -> (Pose) {
    boneToModify {boneName=dyn_hair_17_l}
    constraintSetup [{targetBone {boneName=clavicle_l}}, {targetBone {boneName=neck_02}}, {targetBone {boneName=upperarm_l}}]
    constraintWeights [1,1,1]
   }
   n19 = "Constraint - dyn_hair_21_00_r (PPP)"(ComponentPose=@n18, Alpha=1, ConstraintWeights_0=0, ConstraintWeights_1=1, ConstraintWeights_2=0) {
    boneToModify {boneName=dyn_hair_21_00_r}
    constraintSetup [{targetBone {boneName=clavicle_l}}, {targetBone {boneName=neck_02}}, {targetBone {boneName=upperarm_l}}]
    constraintWeights [1,1,1]
   }
   n20 = "Constraint - dyn_hair_21_01_r (PPP)"(ComponentPose=@n19, Alpha=1, ConstraintWeights_0=0.5, ConstraintWeights_1=0.5, ConstraintWeights_2=0) {
    boneToModify {boneName=dyn_hair_21_01_r}
    constraintSetup [{targetBone {boneName=clavicle_r}}, {targetBone {boneName=neck_02}}, {targetBone {boneName=upperarm_l}}]
    constraintWeights [1,1,1]
   }
   n21 = "Constraint - dyn_Hair_L_02 (PPP)"!(ComponentPose=@n20, Alpha=1, ConstraintWeights_0=0.25, ConstraintWeights_1:float=0.25, ConstraintWeights_2:float=0.5) -> (Pose) {
    boneToModify {boneName=dyn_Hair_L_02}
    constraintSetup [{targetBone {boneName=clavicle_l}}, {targetBone {boneName=spine_05}}, {targetBone {boneName=neck_02}}]
    constraintWeights [1,1,1]
   }
   n22 = "Transform (Modify) Bone - Bone: dyn_Hair_L_01"(ComponentPose=@n21, Alpha=1, Translation="0,0.5,0", Rotation="0,0,0", Scale="1,1,1") {
    rotationMode=BMM_Additive
    rotationSpace=BCS_BoneSpace
    translationMode=BMM_Additive
    translationSpace=BCS_BoneSpace
    boneToModify {boneName=dyn_Hair_L_01}
    translation {y=0.5}
   }
   n23 = "Transform (Modify) Bone - Bone: dyn_Hair_L_02"(ComponentPose=@n22, Alpha=1, Translation="0,0,0", Rotation="0,20,0", Scale="1,1,1") {
    rotationMode=BMM_Additive
    rotationSpace=BCS_BoneSpace
    translationMode=BMM_Additive
    translationSpace=BCS_BoneSpace
    boneToModify {boneName=dyn_Hair_L_02}
    rotation {yaw=20}
   }
   n24 = "Constraint - dyn_Hair_R_02 (PPP)"(ComponentPose=@n23, Alpha=1, ConstraintWeights_0=0, ConstraintWeights_1=1, ConstraintWeights_2=0) {
    boneToModify {boneName=dyn_Hair_R_02}
    constraintSetup [{targetBone {boneName=clavicle_r}}, {targetBone {boneName=spine_05}}, {targetBone {boneName=upperarm_r}}]
    constraintWeights [1,1,1]
   }
   n25 = "Constraint - dyn_Hair_clipping_joint (PP)"!(
    ComponentPose="Transform (Modify) Bone - Bone: dyn_Hair_R_02"(ComponentPose=@n24, Alpha=1, Translation="0,0,0", Rotation="0,10,0", Scale="1,1,1") {rotationMode=BMM_Additive, rotationSpace=BCS_BoneSpace, boneToModify {boneName=dyn_Hair_R_02}, rotation {yaw=10}}
    Alpha=1
    ConstraintWeights_0=0.5
    ConstraintWeights_1:float=0.5
   ) -> (Pose) {
    boneToModify {boneName=dyn_Hair_clipping_joint}
    constraintSetup [{targetBone {boneName=clavicle_r}}, {targetBone {boneName=upperarm_r}}]
    constraintWeights [1,1]
   }
   n26 = "Constraint - dyn_Hair_M_02 (PP)"!(ComponentPose=@n25, Alpha=1, ConstraintWeights_0=1, ConstraintWeights_1:float=0) -> (Pose) {
    boneToModify {boneName=dyn_Hair_M_02}
    constraintSetup [{targetBone {boneName=spine_05}}, {targetBone {boneName=neck_02}}]
    constraintWeights [1,1]
   }
   n27 = "Constraint - dyn_hair_17_l_02 (PP)"!(ComponentPose=@n26, Alpha=1, ConstraintWeights_0=0.3, ConstraintWeights_1:float=0.7) -> (Pose) {
    boneToModify {boneName=dyn_hair_17_l_02}
    constraintSetup [{targetBone {boneName=clavicle_l}}, {targetBone {boneName=neck_02}}]
    constraintWeights [1,1]
   }
   n28 = Component To Local(ComponentPose=@n27)
   n29 = Use cached pose 'Control Rig Pose'()
   n30 = OR Boolean(A=Get isWearingBackbling(), B=Get Is Wearing Cape() -> (Is Wearing Cape:bool))
   n31 = Anim Dynamics - Bone: dyn_hair_01_m(
    ComponentPose=Local To Component(LocalPose=Blend Poses by bool(BlendPose_0=@n28, BlendPose_1=@n29, bActiveValue=@n30) {blendType=Linear})
    Alpha=1
    ExternalForce="0,0,0"
   ) {
    angularDampingOverride=0.99
    angularSpringConstant=1000
    bAngularSpring=true
    bOverrideAngularDamping=true
    bUseGravityOverride=true
    gravityScale=0.35
    simulationSpace=RootRelative
    boundBone {boneName=dyn_hair_01_m}
    componentAppliedLinearAccClamp=<3000 3000 3000>
    componentLinearAccScale=<1 1 1>
    componentLinearVelScale=<1 1 1>
    gravityOverride {z=-980}
    physicsBodyDefinitions [{
      boundBone {boneName=dyn_hair_01_m}
      boxExtents=<20 30 30>
      constraintSetup {angularLimitsMax=<0 15 15>, angularLimitsMin=<0 0 0>, angularTarget=<1 0 0>}
      localJointOffset=<4 0 0>
     }]
    relativeSpaceBone {boneName=head}
   }
   n32 = Anim Dynamics - Bone: base_hair_06_l(ComponentPose=@n31, Alpha=1, ExternalForce="0,0,0") {
    angularDampingOverride=0.99
    angularSpringConstant=5000
    bAngularSpring=true
    bOverrideAngularDamping=true
    bUseGravityOverride=true
    gravityScale=0.35
    simulationSpace=RootRelative
    alphaBoolBlend {blendOption=Linear}
    boundBone {boneName=base_hair_06_l}
    componentAppliedLinearAccClamp=<3000 3000 3000>
    componentLinearAccScale=<0.8 0.8 0.8>
    componentLinearVelScale=<0.8 0.8 0.8>
    gravityOverride {z=-980}
    physicsBodyDefinitions [{
      boundBone {boneName=base_hair_06_l}
      boxExtents=<20 30 30>
      constraintSetup {angularLimitsMax=<0 10 10>, angularLimitsMin=<0 -10 0>, angularTarget=<1 0 0>}
      localJointOffset=<4 0 0>
     }]
   }
   n33 = Anim Dynamics - Bone: dyn_hair_03_l(ComponentPose=@n32, Alpha=1, ExternalForce="0,0,0") {
    angularDampingOverride=0.99
    angularSpringConstant=5000
    bAngularSpring=true
    bOverrideAngularDamping=true
    bUseGravityOverride=true
    gravityScale=0.35
    simulationSpace=RootRelative
    alphaBoolBlend {blendOption=Linear}
    boundBone {boneName=dyn_hair_03_l}
    componentAppliedLinearAccClamp=<3000 3000 3000>
    componentLinearAccScale=<0.8 0.8 0.8>
    componentLinearVelScale=<0.8 0.8 0.8>
    gravityOverride {z=-980}
    physicsBodyDefinitions [{
      boundBone {boneName=dyn_hair_03_l}
      boxExtents=<20 30 30>
      constraintSetup {angularLimitsMax=<0 5 15>, angularLimitsMin=<0 -5 -1>, angularTarget=<1 0 0>}
      localJointOffset=<4 0 0>
     }]
    relativeSpaceBone {boneName=head}
   }
   n34 = Anim Dynamics - Bone: dyn_hair_04_r(ComponentPose=@n33, Alpha=1, ExternalForce="0,0,0") {
    angularDampingOverride=0.99
    angularSpringConstant=5000
    bAngularSpring=true
    bOverrideAngularDamping=true
    bUseGravityOverride=true
    gravityScale=0.35
    simulationSpace=RootRelative
    alphaBoolBlend {blendOption=Linear}
    boundBone {boneName=dyn_hair_04_r}
    componentAppliedLinearAccClamp=<3000 3000 3000>
    componentLinearAccScale=<0.8 0.8 0.8>
    componentLinearVelScale=<0.8 0.8 0.8>
    gravityOverride {z=-980}
    physicsBodyDefinitions [{
      boundBone {boneName=dyn_hair_04_r}
      boxExtents=<20 30 30>
      constraintSetup {angularLimitsMax=<0 10 15>, angularLimitsMin=<0 -10 0>, angularTarget=<1 0 0>}
      localJointOffset=<4 0 0>
     }]
    relativeSpaceBone {boneName=head}
   }
   n35 = Anim Dynamics - Bone: dyn_hair_08_r(ComponentPose=@n34, Alpha=1, ExternalForce="0,0,0") {
    angularDampingOverride=0.99
    angularSpringConstant=5000
    bAngularSpring=true
    bOverrideAngularDamping=true
    bUseGravityOverride=true
    gravityScale=0.35
    simulationSpace=RootRelative
    alphaBoolBlend {blendOption=Linear}
    boundBone {boneName=dyn_hair_08_r}
    componentAppliedLinearAccClamp=<3000 3000 3000>
    componentLinearAccScale=<0.8 0.8 0.8>
    componentLinearVelScale=<0.8 0.8 0.8>
    gravityOverride {z=-980}
    physicsBodyDefinitions [{
      boundBone {boneName=dyn_hair_08_r}
      boxExtents=<20 30 30>
      constraintSetup {angularLimitsMax=<0 10 15>, angularLimitsMin=<0 -10 0>, angularTarget=<1 0 0>}
      localJointOffset=<4 0 0>
     }]
    relativeSpaceBone {boneName=head}
   }
   n36 = Anim Dynamics - Bone: dyn_hair_13_r(ComponentPose=@n35, Alpha=1, ExternalForce="0,0,0") {
    angularDampingOverride=0.99
    angularSpringConstant=5000
    bAngularSpring=true
    bOverrideAngularDamping=true
    bUseGravityOverride=true
    gravityScale=0.35
    simulationSpace=RootRelative
    alphaBoolBlend {blendOption=Linear}
    boundBone {boneName=dyn_hair_13_r}
    componentAppliedLinearAccClamp=<3000 3000 3000>
    componentLinearAccScale=<0.8 0.8 0.8>
    componentLinearVelScale=<0.8 0.8 0.8>
    gravityOverride {z=-980}
    physicsBodyDefinitions [{
      boundBone {boneName=dyn_hair_13_r}
      boxExtents=<20 30 30>
      constraintSetup {angularLimitsMax=<0 5 15>, angularLimitsMin=<0 -5 0>, angularTarget=<1 0 0>}
      localJointOffset=<4 0 0>
     }]
    relativeSpaceBone {boneName=head}
   }
   n37 = Anim Dynamics - Bone: dyn_hair_15_l(ComponentPose=@n36, Alpha=1, ExternalForce="0,0,0") {
    angularDampingOverride=0.99
    angularSpringConstant=1000
    bAngularSpring=true
    bOverrideAngularDamping=true
    bUseGravityOverride=true
    gravityScale=0.35
    simulationSpace=RootRelative
    boundBone {boneName=dyn_hair_15_l}
    componentAppliedLinearAccClamp=<3000 3000 3000>
    componentLinearAccScale=<0.8 0.8 0.8>
    componentLinearVelScale=<0.8 0.8 0.8>
    gravityOverride {z=-980}
    physicsBodyDefinitions [{
      boundBone {boneName=dyn_hair_15_l}
      boxExtents=<20 30 30>
      constraintSetup {angularLimitsMax=<0 8 10>, angularLimitsMin=<0 0 0>, angularTarget=<1 0 0>}
      localJointOffset=<1 0 0>
     }]
    relativeSpaceBone {boneName=head}
   }
   n38 = Anim Dynamics - Bone: dyn_hair_01_r(ComponentPose=@n37, Alpha=1, ExternalForce="0,0,0") {
    angularDampingOverride=0.99
    angularSpringConstant=1000
    bAngularSpring=true
    bOverrideAngularDamping=true
    bUseGravityOverride=true
    gravityScale=0.35
    simulationSpace=RootRelative
    boundBone {boneName=dyn_hair_01_r}
    componentAppliedLinearAccClamp=<3001 3000 3000>
    componentLinearAccScale=<0.8 0.8 0.8>
    componentLinearVelScale=<0.8 0.8 0.8>
    gravityOverride {z=-980}
    physicsBodyDefinitions [{
      boundBone {boneName=dyn_hair_01_r}
      boxExtents=<20 30 30>
      constraintSetup {angularLimitsMax=<0 8 15>, angularLimitsMin=<0 0 0>, angularTarget=<1 0 0>}
      localJointOffset=<4 0 0>
     }]
    relativeSpaceBone {boneName=head}
   }
   n39 = Anim Dynamics - Bone: dyn_earing_l!(ComponentPose=@n38, Alpha=1, GravityOverride:Vector="0,0,-980", ExternalForce="0,0,0") -> (Pose) {
    angularDampingOverride=0.99
    angularSpringConstant=5000
    bAngularSpring=true
    bOverrideAngularDamping=true
    bUseGravityOverride=true
    gravityScale=0.35
    simulationSpace=RootRelative
    boundBone {boneName=dyn_earing_l}
    componentAppliedLinearAccClamp=<3000 3000 3000>
    componentLinearAccScale=<1 1 1>
    componentLinearVelScale=<1 1 1>
    gravityOverride {z=-980}
    physicsBodyDefinitions [{
      boundBone {boneName=dyn_earing_l}
      boxExtents=<20 30 30>
      constraintSetup {angularLimitsMax=<0 30 10>, angularLimitsMin=<0 -40 -10>, angularTarget=<1 0 0>}
      localJointOffset=<4 0 0>
     }]
    relativeSpaceBone {boneName=spine_05}
   }
   n40 = Anim Dynamics - Bone: dyn_earing_r!(ComponentPose=@n39, Alpha=1, GravityOverride:Vector="0,0,-980", ExternalForce="0,0,0") -> (Pose) {
    angularDampingOverride=0.99
    angularSpringConstant=5000
    bAngularSpring=true
    bOverrideAngularDamping=true
    bUseGravityOverride=true
    gravityScale=0.35
    simulationSpace=RootRelative
    boundBone {boneName=dyn_earing_r}
    componentAppliedLinearAccClamp=<3000 3000 3000>
    componentLinearAccScale=<1 1 1>
    componentLinearVelScale=<1 1 1>
    gravityOverride {z=-980}
    physicsBodyDefinitions [{
      boundBone {boneName=dyn_earing_r}
      boxExtents=<20 30 30>
      constraintSetup {angularLimitsMax=<0 30 10>, angularLimitsMin=<0 -40 -10>, angularTarget=<1 0 0>}
      localJointOffset=<4 0 0>
     }]
    relativeSpaceBone {boneName=spine_05}
   }
   n41 = "Constraint - dyn_hair_21_00_r (PPP)"(ComponentPose=@n40, Alpha=0.2, ConstraintWeights_0=0, ConstraintWeights_1=1, ConstraintWeights_2=0) {
    alphaScaleBiasClamp {clampMax=0.2}
    boneToModify {boneName=dyn_hair_21_00_r}
    constraintSetup [{targetBone {boneName=clavicle_l}}, {targetBone {boneName=neck_02}}, {targetBone {boneName=upperarm_l}}]
    constraintWeights [1,1,1]
   }
   n42 = "Constraint - dyn_hair_21_01_r (PPP)"(ComponentPose=@n41, Alpha=0.2, ConstraintWeights_0=0.5, ConstraintWeights_1=0.5, ConstraintWeights_2=0) {
    alphaScaleBiasClamp {clampMax=0.2}
    boneToModify {boneName=dyn_hair_21_01_r}
    constraintSetup [{targetBone {boneName=clavicle_r}}, {targetBone {boneName=neck_02}}, {targetBone {boneName=upperarm_l}}]
    constraintWeights [1,1,1]
   }
   n43 = "Transform (Modify) Bone - Bone: base_hair_21_r"!(ComponentPose=@n42, bAlphaBoolEnabled:bool=True, Translation="0,0,0", Rotation="0,7,0", Scale="1,1,1") -> (Pose) {
    alphaInputType=Bool
    rotationMode=BMM_Additive
    rotationSpace=BCS_BoneSpace
    scaleMode=BMM_Additive
    scaleSpace=BCS_BoneSpace
    translationMode=BMM_Additive
    translationSpace=BCS_BoneSpace
    alphaBoolBlend {blendInTime=0.2, blendOutTime=0.2}
    boneToModify {boneName=base_hair_21_r}
    rotation {yaw=7}
   }
   Output Pose(Result=Component To Local(ComponentPose=@n43)) {layerGroup=None}
   n44 = Control Rig!(
    Source=Copy Pose From Mesh() {bUseAttachedParent=true}
    Alpha
    is crouching:bool=Get Is Crouching CR() -> (Is Crouching CR:bool)
    is crouch walking:bool=Get Is Crouch walking CR() -> (Is Crouch walking CR:bool)
    is walking2d:bool=Get Is Moving 2D() -> (Is Moving 2D:bool)
    is targeting:bool=Get Is Targeting() -> (Is Targeting:bool)
    is SkyDiving:bool=Get Is Sky Diving() -> (Is Sky Diving:bool)
    is Playing Emote:bool=Get Is Playing Emote() -> (Is Playing Emote:bool)
   ) -> (Pose) {
    alphaBoolBlend {blendOption=Linear}
    controlRigAssetReference {blueprintRigClass=&/F_MED_ChimeCurlTell_FaceAcc_CtrlRig.F_MED_ChimeCurlTell_FaceAcc_CtrlRig_C}
   }
   n45 = RigidBody!(ComponentPose=Local To Component(LocalPose=@n44), Alpha, OverrideWorldGravity:Vector="0,0,-980", ExternalForce) -> (Pose) {
    bClampLinearTranslationLimitToRefPose=true
    bOverrideWorldGravity=true
    overridePhysicsAsset=&/F_MED_ChimeCurlTell_FaceAcc_Physics
    simulationSpace=BaseBoneSpace
    alphaBoolBlend {blendOption=Linear}
    baseBoneRef {boneName=neck_02}
    componentAppliedLinearAccClamp=<0 0 0>
    overrideWorldGravity {z=-980}
    simSpaceSettings {
     dampingAlpha=0
     maxAngularAcceleration=24
     maxAngularVelocity=6
     maxLinearAcceleration=1200
     maxLinearVelocity=300
     worldAlpha=1
     externalLinearDragV=<0.2 0.2 0.2>
    }
   }
   Save cached pose 'Control Rig Pose'(Pose=Component To Local(ComponentPose=@n45)) {cacheName=Control Rig Pose}
  ]}
 EventGraph {nodes [
   Event BlueprintUpdateAnimation()
   > n1 = Sequence()
   > n2 = "Gravity Override State Function (Parameter Asset) (Nativized)"(
    CharacterPartAnimInstance=Self-Reference()
    GameTimeInSeconds=GetGameTimeInSeconds()
    Parameters=&/M_MED_ChimeCurlTell_Parameters
   )
   > Set GravityOverride(GravityOverride=@n2)
   n3 = Get bIsCrouching()
   > Set Is Crouching CR(Is Crouching CR:bool=@n3) -> (Output_Get:bool)
   n4 = Get bIsTargeting()
   > Set Is Crouch walking CR(Is Crouch walking CR:bool=AND Boolean!(A=@n3, B=Get bIsMoving2D(), C:bool=@n4) -> (ReturnValue)) -> (Output_Get:bool)
   > Set Is Moving 2D(Is Moving 2D:bool=false) -> (Output_Get:bool)
   > Set Is Targeting(Is Targeting:bool=@n4) -> (Output_Get:bool)
   > Set Is Sky Diving(Is Sky Diving:bool=Get bIsSkyDiving()) -> (Output_Get:bool)
   > Set Is Playing Emote(Is Playing Emote:bool=Get bIsPlayingEmote()) -> (Output_Get:bool)
   n5 = Get bIsPlayingEmote()
   > Set FEIdle(FEIdle:bool=AND Boolean!(A=Get bIsInFrontEnd(), B=NOT Boolean(A=@n5)) -> (ReturnValue)) -> (Output_Get:bool)
   n6 = CheckForBackbling()
   Set isWearingBackbling(
    execute=@n1.then_1
    isWearingBackbling=AND Boolean!(A=@n6, B=NOT Boolean(A=@n6.bIsBackblingACape), C:bool=NOT Boolean(A=@n6.bIsBackblingLong)) -> (ReturnValue)
   )
   > Set Is Wearing Cape(Is Wearing Cape=OR Boolean(A=@n6.bIsBackblingACape, B=@n6.bIsBackblingLong))
  ]}
]
variables [
 FEIdle {type=boolean}
 Gravity Override Params {type=GravityOverrideParamsStruct}
 GravityOverride {type=Vector}
 Is Crouch walking CR {type=boolean}
 Is Crouching CR {type=boolean}
 Is Moving 2D {type=boolean}
 Is Playing Emote {type=boolean}
 Is Sky Diving {type=boolean}
 Is Targeting {type=boolean}
 Is Wearing Cape {type=boolean}
 isWearingBackbling {type=boolean}
 RBAN Params {
  type=RBANParamsStruct
  default {
   dBNO {simSpaceSettings {
     maxAngularAcceleration=1000
     maxAngularVelocity=3
     maxLinearAcceleration=300
     maxLinearVelocity=700
     velocityScaleZ=0
     externalLinearDragV=<0 0.5 0>
    }}
   dBNO_Carried {simSpaceSettings {
     maxAngularAcceleration=1000
     maxAngularVelocity=3
     maxLinearAcceleration=300
     maxLinearVelocity=700
     velocityScaleZ=0
     externalLinearDragV=<0 0.5 0>
    }}
   emoteOrMelee {simSpaceSettings {
     maxAngularAcceleration=1000
     maxAngularVelocity=3
     maxLinearAcceleration=300
     maxLinearVelocity=700
     velocityScaleZ=0
     externalLinearDragV=<0 0.5 0>
    }}
   front_End {simSpaceSettings {
     maxAngularAcceleration=1000
     maxAngularVelocity=3
     maxLinearAcceleration=300
     maxLinearVelocity=700
     velocityScaleZ=0
     externalLinearDragV=<0 0.5 0>
    }}
   onGround_Crouch_Moving {
    componentLinearAccScale=<0.3 0 0>
    componentLinearVelScale=<0.3 0 0>
    simSpaceSettings {
     dampingAlpha=0
     maxAngularAcceleration=24
     maxAngularVelocity=6
     maxLinearAcceleration=1200
     maxLinearVelocity=300
     velocityScaleZ=0.5
     externalLinearDragV=<0.2 0.2 0.2>
    }
   }
   onGround_Falling {simSpaceSettings {
     maxAngularAcceleration=1000
     maxAngularVelocity=3
     maxLinearAcceleration=300
     maxLinearVelocity=700
     velocityScaleZ=0
     externalLinearDragV=<0 0.5 0>
    }}
   onGround_Moving {simSpaceSettings {
     dampingAlpha=0
     maxAngularAcceleration=24
     maxAngularVelocity=6
     maxLinearAcceleration=1200
     maxLinearVelocity=300
     velocityScaleZ=0.5
     externalLinearDragV=<0.2 0.2 0.2>
    }}
   onGround_Standing {simSpaceSettings {
     dampingAlpha=0
     maxAngularAcceleration=24
     maxAngularVelocity=6
     maxLinearAcceleration=1200
     maxLinearVelocity=300
     velocityScaleZ=0.5
     externalLinearDragV=<0.2 0.2 0.2>
    }}
   rail_Grinding {simSpaceSettings {
     maxAngularAcceleration=1000
     maxAngularVelocity=3
     maxLinearAcceleration=300
     maxLinearVelocity=700
     velocityScaleZ=0
     worldAlpha=0.1
     externalLinearDragV=<0 0.5 0>
    }}
   riding_Moving {simSpaceSettings {
     maxAngularAcceleration=1000
     maxAngularVelocity=3
     maxLinearAcceleration=300
     maxLinearVelocity=700
     velocityScaleZ=0
     externalLinearDragV=<0 0.5 0>
    }}
   riding_Moving_Targeting {simSpaceSettings {
     maxAngularAcceleration=1000
     maxAngularVelocity=3
     maxLinearAcceleration=300
     maxLinearVelocity=700
     velocityScaleZ=0
     externalLinearDragV=<0 0.5 0>
    }}
   riding_Standing {simSpaceSettings {
     maxAngularAcceleration=1000
     maxAngularVelocity=3
     maxLinearAcceleration=300
     maxLinearVelocity=700
     velocityScaleZ=0
     externalLinearDragV=<0 0.5 0>
    }}
   skydive_DivingDown {simSpaceSettings {
     maxAngularAcceleration=1000
     maxAngularVelocity=3
     maxLinearAcceleration=300
     maxLinearVelocity=700
     velocityScaleZ=0
     externalLinearDragV=<0 0.5 0>
    }}
   skydive_FloatingDown {
    componentLinearAccScale=<0 0.5 0>
    componentLinearVelScale=<0 0.5 0>
    simSpaceSettings {
     maxAngularAcceleration=1000
     maxAngularVelocity=3
     maxLinearAcceleration=300
     maxLinearVelocity=700
     velocityScaleZ=0
     externalLinearDragV=<0 0.5 0>
    }
   }
   skydive_GoingUp {simSpaceSettings {
     maxAngularAcceleration=1000
     maxAngularVelocity=3
     maxLinearAcceleration=300
     maxLinearVelocity=700
     velocityScaleZ=0
     externalLinearDragV=<0 0.5 0>
    }}
   skydive_Parachuting {
    componentLinearAccScale=<0 0.1 0.3>
    componentLinearVelScale=<0 0.1 0.3>
    simSpaceSettings {
     maxAngularAcceleration=1000
     maxAngularVelocity=3
     maxLinearAcceleration=300
     maxLinearVelocity=700
     velocityScaleZ=0
     externalLinearDragV=<0 0.5 0>
    }
   }
   skydive_Surfing {simSpaceSettings {
     maxAngularAcceleration=1000
     maxAngularVelocity=3
     maxLinearAcceleration=300
     maxLinearVelocity=700
     velocityScaleZ=0
     externalLinearDragV=<0 0.5 0>
    }}
   swimming {simSpaceSettings {
     maxAngularAcceleration=1000
     maxAngularVelocity=3
     maxLinearAcceleration=300
     maxLinearVelocity=700
     velocityScaleZ=0
     externalLinearDragV=<0 0.5 0>
    }}
   tactical_Sprint {simSpaceSettings {
     maxAngularAcceleration=1000
     maxAngularVelocity=3
     maxLinearAcceleration=300
     maxLinearVelocity=700
     velocityScaleZ=0
     externalLinearDragV=<0 0.5 0>
    }}
   ziplining {simSpaceSettings {
     maxAngularAcceleration=1000
     maxAngularVelocity=3
     maxLinearAcceleration=300
     maxLinearVelocity=700
     velocityScaleZ=0
     externalLinearDragV=<0 0.5 0>
    }}
  }
 }
 RBAN Settings Out {
  type=RBANSettingStruct
  default {
   componentLinearAccScale=<0 0 0>
   componentLinearVelScale=<0 0 0>
   simSpaceSettings {
    dampingAlpha=1
    maxAngularAcceleration=1000
    maxAngularVelocity=3
    maxLinearAcceleration=300
    maxLinearVelocity=700
    velocityScaleZ=0
    worldAlpha=1
    externalAngularVelocity=<0 0 0>
    externalLinearDragV=<0 0.5 0>
    externalLinearVelocity=<0 0 0>
   }
  }
 }
]
